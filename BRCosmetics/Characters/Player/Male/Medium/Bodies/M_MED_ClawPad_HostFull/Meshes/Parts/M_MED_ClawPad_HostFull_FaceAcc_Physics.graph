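# Physics asset for the ClawPad HostFull male medium face accessory mesh, providing capsule bodies for its dynamic hood bones.
# Left, right and rear hood chains hang from the head through tiny MSTR root bodies, with tight angular limits per joint.
# Larger capsules cover the head, upper spine and clavicles.
# Free-rotation constraints cross-link side chain tips, plus one right tip to the rear chain, with no left equivalent.
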PhysicsAsset /BRCosmetics/Characters/Player/Male/Medium/Bodies/M_MED_ClawPad_HostFull/Meshes/Parts/M_MED_ClawPad_HostFull_FaceAcc_Physics
bodies {
 clavicle_l [{
   shapeType=Capsule
   length=9.1844
   radius=9.7471
   shapeName=clavicle_l_capsule
   center=<9.8257 0 -1.5556>
   rotation=<rot 82.9197 9.9308 -5.7847>
  }]
 clavicle_r [{
   shapeType=Capsule
   length=9.1844
   radius=9.7471
   shapeName=clavicle_r_capsule
   center=<-9.8257 0 1.5556>
   rotation=<rot 82.9197 9.9308 174.2153>
  }]
 dyn_hood_l_01 [{shapeType=Capsule, length=5, radius=4, shapeName=dyn_hood_l_01_capsule, center=<4 0 5.5>, rotation=<rot 0 0 -90>}]
 dyn_hood_l_01_MSTR [{shapeType=Capsule, length=0.505, radius=0.505, shapeName=dyn_hood_l_01_MSTR_capsule, rotation=<rot 0 -90 0>}]
 dyn_hood_l_01_end [{shapeType=Capsule, length=13.6875, radius=4.0625, shapeName=dyn_hood_l_01_end_capsule, center=<2 0 3>, rotation=<rot -0.0016 0.028 -61.3309>}]
 dyn_hood_l_02 [{shapeType=Capsule, length=5, radius=4, shapeName=dyn_hood_l_02_capsule, center=<6.5 0 4>, rotation=<rot 0 0 -90>}]
 dyn_hood_l_02_MSTR [{shapeType=Capsule, length=0.505, radius=0.505, shapeName=dyn_hood_l_02_MSTR_capsule, rotation=<rot 0 -90 0>}]
 dyn_hood_l_02_end [{shapeType=Capsule, length=8.4375, radius=5.9375, shapeName=dyn_hood_l_02_end_capsule, center=<5.5 0 3>, rotation=<rot 0 0 -90>}]
 dyn_hood_r_01 [{shapeType=Capsule, length=5, radius=4, shapeName=dyn_hood_r_01_capsule, center=<-4 0 -5.5>, rotation=<rot 0 0 90>}]
 dyn_hood_r_01_MSTR [{shapeType=Capsule, length=0.505, radius=0.505, shapeName=dyn_hood_r_01_MSTR_capsule, rotation=<rot 0 -90 0>}]
 dyn_hood_r_01_end [{shapeType=Capsule, length=13.6875, radius=4.0625, shapeName=dyn_hood_r_01_end_capsule, center=<-2 0 -3>, rotation=<rot 0 0.0001 116.897>}]
 dyn_hood_r_02 [{shapeType=Capsule, length=5, radius=4, shapeName=dyn_hood_r_02_capsule, center=<-6.5 0 -4>, rotation=<rot 0 0 90>}]
 dyn_hood_r_02_MSTR [{shapeType=Capsule, length=0.505, radius=0.505, shapeName=dyn_hood_r_02_MSTR_capsule, rotation=<rot 0 -90 0>}]
 dyn_hood_r_02_end [{shapeType=Capsule, length=8.4375, radius=5.9375, shapeName=dyn_hood_r_02_end_capsule, center=<-5.5 0 -3>, rotation=<rot 0 0 90>}]
 dyn_hood_rear_01 [{shapeType=Capsule, length=3.5625, radius=5.4375, shapeName=dyn_hood_rear_01_capsule, center=<4 0 6.5>, rotation=<rot 0 0 -90>}]
 dyn_hood_rear_02 [{shapeType=Capsule, length=13.8735, radius=8.0625, shapeName=dyn_hood_rear_02_capsule, center=<6.9952 0 4.7259>, rotation=<rot 0 0 -90>}]
 dyn_hood_rear_03 [{shapeType=Capsule, length=10.25, radius=11, shapeName=dyn_hood_rear_03_capsule, center=<10.7101 0 4.5>, rotation=<rot -0.0001 0 -86.1111>}]
 dyn_hood_rear_MSTR [{shapeType=Capsule, length=0.505, radius=0.505, shapeName=dyn_hood_rear_MSTR_capsule, rotation=<rot 0 -90 0>}]
 head [{shapeType=Capsule, length=10.875, radius=6.625, rotation=<rot 90 0 0>}]
 spine_05 [
  {shapeType=Capsule, length=7.2096, radius=16.4243, shapeName=spine_05_capsule, center=<-6.3513 1.8849 0>, rotation=<rot 0 -90 0>}
  {shapeType=Capsule, length=3.5625, radius=11.1875, shapeName=spine_05_capsule, center=<7.6605 2.6617 0>, rotation=<rot 90 90 63.8097>}
 ]
}
constraints [
 {
  bone1Name=head
  bone2Name=spine_05
  swing1LimitDegrees=45
  swing1Motion=Free
  swing2LimitDegrees=45
  swing2Motion=Free
  twistLimitDegrees=45
  twistMotion=Free
 }
 {
  bone1Name=clavicle_l
  bone2Name=spine_05
  swing1LimitDegrees=45
  swing1Motion=Free
  swing2LimitDegrees=45
  swing2Motion=Free
  twistLimitDegrees=45
  twistMotion=Free
 }
 {bone1Name=dyn_hood_l_01_MSTR, bone2Name=head, swing1LimitDegrees=12, swing2LimitDegrees=12, twistLimitDegrees=12}
 {bone1Name=dyn_hood_l_01, bone2Name=dyn_hood_l_01_MSTR, swing1LimitDegrees=12, swing2LimitDegrees=12, twistLimitDegrees=12}
 {bone1Name=dyn_hood_l_01_end, bone2Name=dyn_hood_l_01, swing1LimitDegrees=12, swing2LimitDegrees=12, twistLimitDegrees=12}
 {bone1Name=dyn_hood_l_02_MSTR, bone2Name=head, swing1LimitDegrees=12, swing2LimitDegrees=12, twistLimitDegrees=12}
 {bone1Name=dyn_hood_l_02, bone2Name=dyn_hood_l_02_MSTR, swing1LimitDegrees=4, swing2LimitDegrees=12, twistLimitDegrees=4}
 {bone1Name=dyn_hood_l_02_end, bone2Name=dyn_hood_l_02, swing1LimitDegrees=4, swing2LimitDegrees=12, twistLimitDegrees=4}
 {bone1Name=dyn_hood_r_01_MSTR, bone2Name=head, swing1LimitDegrees=12, swing2LimitDegrees=12, twistLimitDegrees=12}
 {bone1Name=dyn_hood_r_01, bone2Name=dyn_hood_r_01_MSTR, swing1LimitDegrees=12, swing2LimitDegrees=12, twistLimitDegrees=12}
 {bone1Name=dyn_hood_r_01_end, bone2Name=dyn_hood_r_01, swing1LimitDegrees=12, swing2LimitDegrees=12, twistLimitDegrees=12}
 {bone1Name=dyn_hood_r_02_MSTR, bone2Name=head, swing1LimitDegrees=12, swing2LimitDegrees=12, twistLimitDegrees=12}
 {bone1Name=dyn_hood_r_02, bone2Name=dyn_hood_r_02_MSTR, swing1LimitDegrees=4, swing2LimitDegrees=12, twistLimitDegrees=4}
 {bone1Name=dyn_hood_r_02_end, bone2Name=dyn_hood_r_02, swing1LimitDegrees=4, swing2LimitDegrees=12, twistLimitDegrees=4}
 {bone1Name=dyn_hood_rear_MSTR, bone2Name=head, swing1LimitDegrees=12, swing2LimitDegrees=12, twistLimitDegrees=12}
 {bone1Name=dyn_hood_rear_01, bone2Name=dyn_hood_rear_MSTR, swing1LimitDegrees=12, swing2LimitDegrees=12, twistLimitDegrees=12}
 {bone1Name=dyn_hood_rear_02, bone2Name=dyn_hood_rear_01, swing1LimitDegrees=12, swing2LimitDegrees=12, twistLimitDegrees=12}
 {bone1Name=dyn_hood_rear_03, bone2Name=dyn_hood_rear_02, swing1LimitDegrees=12, swing2LimitDegrees=12, twistLimitDegrees=12}
 {
  bone1Name=clavicle_r
  bone2Name=spine_05
  swing1LimitDegrees=45
  swing1Motion=Free
  swing2LimitDegrees=45
  swing2Motion=Free
  twistLimitDegrees=45
  twistMotion=Free
 }
 {
  bone1Name=dyn_hood_r_02_end
  bone2Name=dyn_hood_r_01_end
  swing1LimitDegrees=45
  swing1Motion=Free
  swing2LimitDegrees=45
  swing2Motion=Free
  twistLimitDegrees=45
  twistMotion=Free
 }
 {
  bone1Name=dyn_hood_l_02_end
  bone2Name=dyn_hood_l_01_end
  swing1LimitDegrees=45
  swing1Motion=Free
  swing2LimitDegrees=45
  swing2Motion=Free
  twistLimitDegrees=45
  twistMotion=Free
 }
 {
  bone1Name=dyn_hood_r_02_end
  bone2Name=dyn_hood_rear_02
  swing1LimitDegrees=45
  swing1Motion=Free
  swing2LimitDegrees=45
  swing2Motion=Free
  twistLimitDegrees=45
  twistMotion=Free
 }
]
